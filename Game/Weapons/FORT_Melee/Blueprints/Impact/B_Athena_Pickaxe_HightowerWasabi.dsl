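; Blueprint for the dual-wield HightowerWasabi pickaxe that drives its Niagara trail, blade flare and impact effects.
; Equipping spawns delayed trails on both weapon meshes unless the pawn is moving fast, falling or jumping; the second left swing does so unconditionally, and the second right swing spawns a blade flare.
; Swing events track side and final-swing state, which impact FX use to rotate along the hit normal, set SecondImpact and add a directional burst.
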(event FortWeapon|EventOnEquippedAndAttached
  ;[1 1 1] Equip trail FX
  (|CallParentFunction)
  (bind _asplayer_pawn_athena (Utilities|Casting|CastToPlayerPawn_Athena (Game|GetPlayerCharacter 0)))
  (if (> (Variables|/Script/FortniteGame.FortPlayerPawnAthena|GetCurrentPawnSquaredSpeed _asplayer_pawn_athena) 8.1e+05)
    _
    (elif (or (AI|Components|NavMovement|IsFalling (Variables|/Script/Engine.Character|GetCharacterMovement _asplayer_pawn_athena)) (Variables|/Script/Engine.Character|GetbWasJumping _asplayer_pawn_athena))
      _
      (else
        (Utilities|FlowControl|RetriggerableDelay 0.09)
        (bind _returnvalue (Niagara|SpawnSystemAttached "/Game/Weapons/FORT_Melee/Pickaxe_HightowerWasabi/FX/NS_Pickaxe_Wasabi_AnimTrail.NS_Pickaxe_Wasabi_AnimTrail" (Variables|Visual|GetWeaponMesh) "None" "0, 0, 0" "0, 0, 0" "KeepRelativeOffset" true false))
        (Niagara|SetNiagaraVariable(Float) _returnvalue "AnimTrailWidth" 1.7)
        (Niagara|SetNiagaraVariable(Float) _returnvalue "Loop duration" 0.13)
        (Components|Activation|SetActive _returnvalue true true)
        (bind _returnvalue_2 (Niagara|SpawnSystemAttached "/Game/Weapons/FORT_Melee/Pickaxe_HightowerWasabi/FX/NS_Pickaxe_Wasabi_AnimTrail.NS_Pickaxe_Wasabi_AnimTrail" (Variables|Visual|GetWeaponMeshOffhand) "None" "0, 0, 0" "0, 0, 0" "KeepRelativeOffset" true false))
        (Niagara|SetNiagaraVariable(Float) _returnvalue_2 "AnimTrailWidth" 1.7)
        (Niagara|SetNiagaraVariable(Float) _returnvalue_2 "Loop duration" 0.13)
        (Components|Activation|SetActive _returnvalue_2 true true)))))

(event EventSwingLeft2
  (|CallParentFunction)
  (Variables|Default|SetLeft2%20Swing? true)
  ;[1 1 1] Final Swing
  (Utilities|FlowControl|RetriggerableDelay 0.17)
  (bind _returnvalue (Niagara|SpawnSystemAttached "/Game/Weapons/FORT_Melee/Pickaxe_HightowerWasabi/FX/NS_Pickaxe_Wasabi_AnimTrail.NS_Pickaxe_Wasabi_AnimTrail" (Variables|Visual|GetWeaponMesh) "None" "0, 0, 0" "0, 0, 0" "KeepRelativeOffset" true false))
  (Niagara|SetNiagaraVariable(Float) _returnvalue "AnimTrailWidth" 1.7)
  (Niagara|SetNiagaraVariable(Float) _returnvalue "Loop duration" 0.132)
  (Components|Activation|SetActive _returnvalue true true)
  (bind _returnvalue_2 (Niagara|SpawnSystemAttached "/Game/Weapons/FORT_Melee/Pickaxe_HightowerWasabi/FX/NS_Pickaxe_Wasabi_AnimTrail.NS_Pickaxe_Wasabi_AnimTrail" (Variables|Visual|GetWeaponMeshOffhand) "None" "0, 0, 0" "0, 0, 0" "KeepRelativeOffset" true false))
  (Niagara|SetNiagaraVariable(Float) _returnvalue_2 "AnimTrailWidth" 1.7)
  (Niagara|SetNiagaraVariable(Float) _returnvalue_2 "Loop duration" 0.132)
  (Components|Activation|SetActive _returnvalue_2 true true))

(event FortWeapon|EventOnPlayImpactFX
  (params (HitResult :type Hit%20Result%20Structure%20%28by%20ref%29) (ImpactPhysicalSurface :type EPhysicalSurface%20Enum :default "SurfaceType_Default") (SpawnedPSC :type FXSystem%20Component%20Object%20Reference))
  (bind (_bblockinghit _binitialoverlap _time _distance _location _impactpoint _normal) (Collision|BreakHitResult HitResult))
  (bind _returnvalue (Math|Rotator|MakeRotfromX _normal))
  (|CallParentFunction HitResult ImpactPhysicalSurface SpawnedPSC)
  (bind _asniagara_particle_system_component (Utilities|Casting|CastToNiagaraComponent SpawnedPSC))
  (Utilities|FlowControl|Sequence
    (:then_0
      (Variables|Default|SetImpact%20FX_0 _asniagara_particle_system_component)
      (Transformation|SetWorldRotation SpawnedPSC (Math|Rotator|CombineRotators _returnvalue (select (Variables|Default|GetRight?) (Variables|Default|GetRight%20Rotation) (Variables|Default|GetLeft%20Rotation))) false false)
      (if (Variables|Default|GetLeft2%20Swing?)
        (bind _returnvalue_2 (Niagara|SpawnSystemAttached "/Game/Weapons/FORT_Melee/Pickaxe_HightowerWasabi/FX/NS_Pickaxe_Wasabi_Impact_Directional.NS_Pickaxe_Wasabi_Impact_Directional" (Variables|Default|GetImpact%20FX_0) "None" "0, 0, 0" "0, 0, 0" "KeepRelativeOffset" true))
        (Transformation|SetWorldRotation _returnvalue_2 (Math|Rotator|CombineRotators _returnvalue "0,60.000000, 0") false false)))
    (:then_1
      (Niagara|SetNiagaraVariable(Bool) _asniagara_particle_system_component "SecondImpact" (Variables|Default|GetLeft2%20Swing?)))))

(event EventSwingLeft_Common
  (|CallParentFunction)
  (Variables|Default|SetRight? false))

(event EventSwingRight_Common
  (|CallParentFunction)
  (Variables|Default|SetRight? true))

(event EventSwingLeft
  (|CallParentFunction)
  (Variables|Default|SetLeft2%20Swing? false))

(event EventSwingRight
  (|CallParentFunction)
  (Variables|Default|SetLeft2%20Swing? false))

(event EventSwingRight2
  ;[1 1 1] Up swing blade Flare FX
  (|CallParentFunction)
  (Variables|Default|SetLeft2%20Swing? false)
  (bind _returnvalue (Niagara|SpawnSystemAttached "/Game/Weapons/FORT_Melee/Pickaxe_HightowerWasabi/FX/NS_Pickaxe_Wasabi_Idle.NS_Pickaxe_Wasabi_Idle" (Variables|Visual|GetWeaponMesh) "FX_Idle_Start" "0, 0, 0" "0, 0, 0" "KeepRelativeOffset" true false))
  (Niagara|SetNiagaraVariable(Float) _returnvalue "Blade_flash_loop" 0.0)
  (Niagara|SetNiagaraVariable(Float) _returnvalue "Blade_flash_attack" 1.0)
  (Components|Activation|SetActive _returnvalue true true))

(fn ConstructionScript
  (|CallParentFunction))
